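# Lobby nameplate banner widget that stacks a background, icon, party-leader icon and voice-chat outline in an overlay.
# PreConstruct caches the outline's dynamic material, Construct collapses the leader icon, and UpdateIconAndBackgroundImages swaps icon and background materials.
# PlayTalkingAnim sets outline visibility and, unless Hidden, endlessly loops Anim_FakeVoicePitch, which fakes voice levels by keying the material's Volume parameter.
# UpdateVoicePitch can also set Volume directly.
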WidgetBlueprint /Game/UI/Frontend/Lobby/Nameplates/WBP_Nameplate_Banner
animations [Anim_FakeVoicePitch {
  binding_count=1
  length_seconds=2
  object=&self:Anim_FakeVoicePitch
  sequence {bindings [VoiceChatOutline {
     id=#1
     possessed_class=Image
     tracks [{
       type=MovieSceneWidgetMaterialTrack
       sections [{
         type=MovieSceneParameterSection
         channels [Volume {
           type=MovieSceneScriptingFloatChannel
           keys [[0,1],[2,0],[3.2,2],[6,0.75],[6.4,0.4],[9,1],[12,0.2],[15,1],[19,0.2],[23,0.5],[26,1],[29,0],[32,0.25],[36,0.1],[40,1]]
          }]
        }]
      }]
    }]}
 }]
graphs [
 EventGraph {nodes [
   Event PreConstruct()
   > n1 = GetDynamicMaterial(self=Get VoiceChatOutline() -> (VoiceChatOutline:Image Object Reference))
   > Set MID_VoiceChat(MID_VoiceChat:Material Instance Dynamic Object Reference=@n1) -> (Output_Get:Material Instance Dynamic Object Reference)
   Event Construct()
   > SetVisibility(self=Get Image_LeaderIcon() -> (Image_LeaderIcon:Image Object Reference), InVisibility=Collapsed)
  ]}
 PlayTalkingAnim {nodes [
   n1 = PlayTalkingAnim!() -> (then, VoiceVisibility:ESlateVisibility Enum=Visible) {is=title K2Node_FunctionEntry}
   > Set Visibility(Visibility=@n1, self=Get VoiceChatOutline())
   > n2 = Branch(Condition="Equal (Enum)"(A:ESlateVisibility Enum=@n1, B:ESlateVisibility Enum=Hidden))
   n3 = Get Anim_FakeVoicePitch() -> (Anim_FakeVoicePitch:Widget Animation Object Reference)
   > SetAnimationCurrentTime!(execute, self:User Widget Object Reference, InAnimation:Widget Animation Object Reference=@n3, InTime:float=0) -> (then)
   PlayAnimation(execute=@n2.else, InAnimation=@n3, NumLoopsToPlay=0)
  ]}
 UpdateIconAndBackgroundImages {nodes [
   n1 = UpdateIconAndBackgroundImages!() -> (
    then
    IconAndBacgroundMaterials_Icon_22_#2:Material Instance Object Reference
    IconAndBacgroundMaterials_Background_23_#3:Material Instance Object Reference
   ) {is=title K2Node_FunctionEntry}
   > SetBrushFromMaterial(self=Get Icon() -> (Icon:Image Object Reference), Material=@n1)
   > SetBrushFromMaterial(
    self=Get Background() -> (Background:Image Object Reference)
    Material=@n1.IconAndBacgroundMaterials_Background_23_#3
   )
  ]}
 UpdateVoicePitch {nodes [SetScalarParameterValue(
    self=Get MID_VoiceChat() -> (MID_VoiceChat:Material Instance Dynamic Object Reference)
    ParameterName=Volume
    Value=UpdateVoicePitch!() -> (then, VolumeLevel:double=~) {is=title K2Node_FunctionEntry}
   )]}
]
variables [MID_VoiceChat {type=MaterialInstanceDynamic}]
widgets {
 info {inheritedWidgetCount=0, parentClass=&UserWidget, rootWidgetClass=&Overlay, widgetCount=5}
 widgets [
  O_Banner {widgetClassPath=&Overlay, widget=O_Banner}
  Image_LeaderIcon {
   widgetClassPath=&Image
   bIsVariable=true
   parent=O_Banner
   widget=Image_LeaderIcon
   properties {
    brush {
     bIsDynamicallyLoaded=false
     drawAs=Image
     imageType=FullColor
     mirroring=NoMirror
     resourceName=None
     resourceObject=/Game/UI/Frontend/Lobby/Nameplates/Materials/MI_UI_PartyLeader_Icon
     tiling=NoTile
     imageSize=<44 44>
     margin=<ltrb 0 0 0 0>
     outlineSettings {
      bUseBrushTransparency=false
      roundingType=HalfHeightRadius
      width=0
      color {colorUseRule=UseColor_Specified, specifiedColor=<rgba 0 0 0 0>}
      cornerRadii=<0 0 0 0>
     }
     tintColor {colorUseRule=UseColor_Specified, specifiedColor=<rgba 1 1 1 1>}
     uVRegion {bIsValid=false, max=<0 0>, min=<0 0>}
    }
    renderTransform {angle=0, scale=<1 1>, shear=<0 0>, translation=<0 -8>}
   }
   slot {class=OverlaySlot, properties {horizontalAlignment=HAlign_Center, padding=<ltrb 0 -12 0 0>}}
  }
  VoiceChatOutline {
   widgetClassPath=&Image
   bIsVariable=true
   parent=O_Banner
   widget=VoiceChatOutline
   properties {
    visibility=SelfHitTestInvisible
    brush {
     bIsDynamicallyLoaded=false
     drawAs=Image
     imageType=FullColor
     mirroring=NoMirror
     resourceName=None
     resourceObject=/Game/UI/Frontend/Lobby/Nameplates/Materials/MI_UI_Nameplate_VoiceChatOutline
     tiling=NoTile
     imageSize=<56 56>
     margin=<ltrb 0 0 0 0>
     outlineSettings {
      bUseBrushTransparency=false
      roundingType=HalfHeightRadius
      width=0
      color {colorUseRule=UseColor_Specified, specifiedColor=<rgba 0 0 0 0>}
      cornerRadii=<0 0 0 0>
     }
     tintColor {colorUseRule=UseColor_Specified, specifiedColor=<rgba 1 1 1 1>}
     uVRegion {bIsValid=false, max=<0 0>, min=<0 0>}
    }
   }
   slot {class=OverlaySlot, properties {horizontalAlignment=HAlign_Center, verticalAlignment=VAlign_Center}}
  }
  Background {
   widgetClassPath=&Image
   bIsVariable=true
   parent=O_Banner
   widget=Background
   properties {
    visibility=SelfHitTestInvisible
    brush {
     bIsDynamicallyLoaded=false
     drawAs=Image
     imageType=FullColor
     mirroring=NoMirror
     resourceName=None
     resourceObject=/Game/UI/Frontend/Lobby/Nameplates/Materials/MI_UI_Nameplate_BannerBackground
     tiling=NoTile
     imageSize=<48 48>
     margin=<ltrb 0 0 0 0>
     outlineSettings {
      bUseBrushTransparency=false
      roundingType=HalfHeightRadius
      width=0
      color {colorUseRule=UseColor_Specified, specifiedColor=<rgba 0 0 0 0>}
      cornerRadii=<0 0 0 0>
     }
     tintColor {colorUseRule=UseColor_Specified, specifiedColor=<rgba 1 1 1 1>}
     uVRegion {bIsValid=false, max=<0 0>, min=<0 0>}
    }
   }
   slot {class=OverlaySlot, properties {horizontalAlignment=HAlign_Center, verticalAlignment=VAlign_Center}}
  }
  Icon {
   widgetClassPath=&Image
   bIsVariable=true
   parent=O_Banner
   widget=Icon
   properties {
    visibility=SelfHitTestInvisible
    brush {
     bIsDynamicallyLoaded=false
     drawAs=Image
     imageType=FullColor
     mirroring=NoMirror
     resourceName=None
     resourceObject=/Game/UI/Frontend/Lobby/Nameplates/Materials/M_UI_Nameplate_BannerIcon
     tiling=NoTile
     imageSize=<48 48>
     margin=<ltrb 0 0 0 0>
     outlineSettings {
      bUseBrushTransparency=false
      roundingType=HalfHeightRadius
      width=0
      color {colorUseRule=UseColor_Specified, specifiedColor=<rgba 0 0 0 0>}
      cornerRadii=<0 0 0 0>
     }
     tintColor {colorUseRule=UseColor_Specified, specifiedColor=<rgba 1 1 1 1>}
     uVRegion {bIsValid=false, max=<0 0>, min=<0 0>}
    }
   }
   slot {class=OverlaySlot, properties {horizontalAlignment=HAlign_Center, verticalAlignment=VAlign_Center}}
  }
 ]
}
