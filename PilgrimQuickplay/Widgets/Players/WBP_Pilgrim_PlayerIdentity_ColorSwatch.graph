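# Swatch widget laying out a player's identity colors in a grid with Avatar, Squad, Banner and Festival Local MP rows.
# On Construct, it tints the squad color, squad marker and local MP images using colors looked up by the owning player's splitscreen index.
# Banner helpers, uncalled in these graphs, return either dynamic nameplate materials or a blocking-loaded Habanero rank icon.
# The avatar and banner widgets receive no values here.
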WidgetBlueprint /PilgrimQuickplay/Widgets/Players/WBP_Pilgrim_PlayerIdentity_ColorSwatch
graphs [
 Banner_GetRankOrBannerMaterial {nodes [
   n1 = Banner_GetRankOrBannerMaterial!() -> (
    then
    IsIslandRanked:bool=~
    BannerIcon:Texture 2D Object Reference
    IconColor:Linear Color Structure
    IsPlayerRanked:bool=~
    RankType:String
    CurrentRank:int
   ) {is=title K2Node_FunctionEntry}
   > n2 = Branch(Condition=@n1)
   > n3 = Branch(Condition=@n1.IsPlayerRanked)
   > n4 = Banner_GetRankTypeIconMaterial!(execute, self, Rank Type:"String (by ref)"=Get RankType() -> (RankType:String), Current Rank:int=Get CurrentRank() -> (CurrentRank:int=0)) -> (then, Material:Material Instance Object Reference, B Found Material:bool)
   > Branch(Condition=@n4.B Found Material)
   > Return Node!(
    execute
    MaterialInstance_Icon_22_#1:Material Instance Object Reference=@n4
    MaterialInstance_Background_23_#2:Material Instance Object Reference
   )
   n5 = CreateDynamicMaterialInstance!(
    execute=@n3.else
    Parent=/Game/UI/Frontend/Lobby/Habanero/Habanero_Materials/MaterialInstances/Icons/Foreground/MI_UI_Habanero_Foreground_00
    OptionalName=None
    CreationFlags=None
   ) -> (then, ReturnValue)
   > Return Node(MaterialInstance_Icon_22_#1=@n5)
   n6 = CreateDynamicMaterialInstance(execute=@n2.else, Parent=/Game/UI/Frontend/Lobby/Nameplates/Materials/MI_UI_Nameplate_BannerIcon, OptionalName=None, CreationFlags=None)
   > n7 = CreateDynamicMaterialInstance(Parent=/Game/UI/Frontend/Lobby/Nameplates/Materials/MI_UI_Nameplate_BannerBackground, OptionalName=None, CreationFlags=None)
   > SetTextureParameterValue(self=@n6, ParameterName=Icon Texture, Value=@n1.BannerIcon)
   > SetVectorParameterValue(self=@n7, ParameterName=Banner Color, Value=@n1.IconColor)
   > Return Node(MaterialInstance_Icon_22_#1=@n6, MaterialInstance_Background_23_#2=@n7)
  ]}
 Banner_GetRankTypeIconMaterial {nodes [
   n1 = Banner_GetRankTypeIconMaterial!() -> (then, Rank Type:"String (by ref)", Current Rank:int) {is=title K2Node_FunctionEntry}
   n2 = Find(
    TargetMap:Map of Strings to Fort Habanero Mode List Structures=Get EntryData!(self:Fort Habanero Display Data Object Reference=Get RankInfo() -> (RankInfo:Fort Habanero Display Data Object Reference)) -> (EntryData:Map of Strings to Fort Habanero Mode List Structures)
    Key=@n1
   ) -> (Value:Fort Habanero Mode List Structure)
   > n3 = Branch(Condition=@n2.ReturnValue)
   n4 = Break Fort Habanero Mode List!(FortHabaneroModeList:"Fort Habanero Mode List Structure (by ref)"=@n2) -> (TierList:Array of Fort Habanero Tier Structures, bShowTopLeaderboardIcon:bool=~, TopTierTopLeaderboardPositions:Fort Habanero Tier Structure) {is=title K2Node_BreakStruct}
   > n5 = Branch(Condition=Is Valid Index!(TargetArray:Array of Fort Habanero Tier Structures=@n4, IndexToTest:int=@n1.Current Rank) -> (ReturnValue:bool))
   n6 = Break Fort Habanero Tier!(FortHabaneroTier:"Fort Habanero Tier Structure (by ref)"="Get (a copy)"(Array:Array of Fort Habanero Tier Structures=@n4, Dimension 1=@n1.Current Rank) -> (Output:Fort Habanero Tier Structure)) -> (
    ForegroundIcon:Material Instance Soft Object Reference
    ForegroundIconSmall:Material Instance Soft Object Reference
    BackgroundIcon:Material Instance Soft Object Reference
    BackgroundIconSmall:Material Instance Soft Object Reference
    WhiteIcon:Texture 2D Soft Object Reference
    SimpleIconSmall:Texture 2D Soft Object Reference
    Name:Text
    TierType:Fort Habanero Tier Type Data Object Reference
    AdditionalLeftIcon:Material Instance Soft Object Reference
    AdditionalRightIcon:Material Instance Soft Object Reference
   ) {is=title K2Node_BreakStruct}
   > n7 = Branch(Condition=IsValid(Object=@n6.TierType))
   > n8 = LoadAsset_Blocking(Asset=@n6.ForegroundIconSmall)
   > n9 = Cast To MaterialInstance!(execute, Object=@n8) -> (then, CastFailed, AsMaterial Instance:Material Instance Object Reference) {is=title K2Node_DynamicCast}
   > Return Node!(execute, Material:Material Instance Object Reference=@n9, B Found Material:bool=true)
   Return Node(execute=[@n9.CastFailed, @n7.else, @n5.else, @n3.else], B Found Material=false)
  ]}
 EventGraph {nodes [
   Event Construct()
   > Squad_SetColor()
   n1 = GetSplitscreenIndex!(PC:Player Controller Object Reference=GetOwningPlayer()) -> (ReturnValue:int=0)
   > SetBrushTintColor!(
    execute
    self:Image Object Reference=Get LocalMP_Color() -> (LocalMP_Color:Image Object Reference)
    TintColor:Slate Color Structure="To SlateColor (Color)"!(Value:Linear Color Structure=GetPilgrim LocalPlayerIdentity Color By Index!(LocalPlayerIndex:int=@n1) -> (Player Identity Color:Linear Color Structure)) -> (ReturnValue:Slate Color Structure)
   ) -> (then)
  ]}
 Squad_SetColor {nodes [
   Squad_SetColor() {is=title K2Node_FunctionEntry}
   n1 = GetSplitscreenIndex(PC=GetOwningPlayer()) -> (ReturnValue=0)
   > SetBrushTintColor(
    self=[Get Squad_Marker() -> (Squad_Marker:Image Object Reference), Get Squad_Color() -> (Squad_Color:Image Object Reference)]
    TintColor="To SlateColor (Color)"(Value=GetPinColorBySquadMemberIndex!(SquadIndex:int=@n1) -> (ReturnValue:Linear Color Structure))
   )
  ]}
]
variables [RankInfo {type=FortHabaneroDisplayData, default=/Game/UI/Habanero/HabaneroDisplayData}]
widgets {
 info {inheritedWidgetCount=0, parentClass=&UserWidget, rootWidgetClass=&GridPanel, widgetCount=18}
 widgets [
  GridPanel_90 {widgetClassPath=&GridPanel, widget=GridPanel_90, properties {columnFill [1,0.5,0.5,0.5], rowFill [1,1,1,1]}}
  Bg {
   widgetClassPath=&Image
   parent=GridPanel_90
   widget=Bg
   properties {brush {
     bIsDynamicallyLoaded=false
     drawAs=Image
     imageType=NoImage
     mirroring=NoMirror
     resourceName=None
     resourceObject=null
     tiling=NoTile
     imageSize=<32 32>
     margin=<ltrb 0 0 0 0>
     outlineSettings {
      bUseBrushTransparency=false
      roundingType=HalfHeightRadius
      width=0
      color {colorUseRule=UseColor_Specified, specifiedColor=<rgba 0 0 0 0>}
      cornerRadii=<0 0 0 0>
     }
     tintColor {colorUseRule=UseColor_Specified, specifiedColor=<rgba 0 0 0 1>}
     uVRegion {bIsValid=false, max=<0 0>, min=<0 0>}
    }}
   slot {class=GridSlot, properties {columnSpan=4, layer=-1, rowSpan=4}}
  }
  UEFN_TextBlock_C_48 {
   widgetClassPath=&UEFN_TextBlock
   parent=GridPanel_90
   widget=UEFN_TextBlock_C_48
   properties {
    text=Avatar
    font {
     bForceMonospaced=false
     bMaterialIsStencil=false
     fontMaterial=null
     fontObject=&BurbankSmall
     letterSpacing=0
     monospacedWidth=1
     size=15
     skewAmount=0
     typefaceFontName=Black
     outlineSettings {
      bApplyOutlineToDropShadows=false
      bMiteredCorners=false
      bSeparateFillAlpha=false
      outlineMaterial=null
      outlineSize=0
      outlineColor=<rgba 0 0 0 1>
     }
    }
   }
   slot {class=GridSlot, properties {columnSpan=1, rowSpan=1, verticalAlignment=VAlign_Center, padding=<ltrb 5 5 5 5>}}
  }
  Avatar_Background {
   widgetClassPath=&Image
   bIsVariable=true
   parent=GridPanel_90
   widget=Avatar_Background
   slot {class=GridSlot, properties {column=1, columnSpan=1, rowSpan=1, padding=<ltrb 5 5 5 5>}}
  }
  Avatar_Highlight {
   widgetClassPath=&Image
   bIsVariable=true
   parent=GridPanel_90
   widget=Avatar_Highlight
   slot {class=GridSlot, properties {column=2, columnSpan=1, rowSpan=1, padding=<ltrb 5 5 5 5>}}
  }
  ScaleBox_Avatar_Icon {
   widgetClassPath=&ScaleBox
   parent=GridPanel_90
   widget=ScaleBox_Avatar_Icon
   slot {class=GridSlot, properties {column=3, columnSpan=1, rowSpan=1, padding=<ltrb 5 5 5 5>}}
  }
  Avatar_Icon {
   widgetClassPath=/PilgrimCore/Widgets/Player/WBP_Pilgrim_AvatarIcon
   bIsVariable=true
   parent=ScaleBox_Avatar_Icon
   widget=Avatar_Icon
   slot {class=ScaleBoxSlot}
  }
  UEFN_TextBlock_C {
   widgetClassPath=&UEFN_TextBlock
   parent=GridPanel_90
   widget=UEFN_TextBlock_C
   properties {
    text=Squad
    font {
     bForceMonospaced=false
     bMaterialIsStencil=false
     fontMaterial=null
     fontObject=&BurbankSmall
     letterSpacing=0
     monospacedWidth=1
     size=15
     skewAmount=0
     typefaceFontName=Black
     outlineSettings {
      bApplyOutlineToDropShadows=false
      bMiteredCorners=false
      bSeparateFillAlpha=false
      outlineMaterial=null
      outlineSize=0
      outlineColor=<rgba 0 0 0 1>
     }
    }
   }
   slot {class=GridSlot, properties {columnSpan=1, row=1, rowSpan=1, verticalAlignment=VAlign_Center, padding=<ltrb 5 5 5 5>}}
  }
  Squad_Color {
   widgetClassPath=&Image
   bIsVariable=true
   parent=GridPanel_90
   widget=Squad_Color
   slot {class=GridSlot, properties {column=1, columnSpan=1, row=1, rowSpan=1, padding=<ltrb 5 5 5 5>}}
  }
  ScaleBox_Squad_Marker {
   widgetClassPath=&ScaleBox
   parent=GridPanel_90
   widget=ScaleBox_Squad_Marker
   slot {class=GridSlot, properties {column=3, columnSpan=1, row=1, rowSpan=1, padding=<ltrb 5 5 5 5>}}
  }
  Squad_Marker {
   widgetClassPath=&Image
   bIsVariable=true
   parent=ScaleBox_Squad_Marker
   widget=Squad_Marker
   properties {
    bIsVolatile=true
    visibility=SelfHitTestInvisible
    brush {
     bIsDynamicallyLoaded=false
     drawAs=Image
     imageType=NoImage
     mirroring=NoMirror
     resourceName=None
     resourceObject=/Game/Athena/HUD/Art/T_UI_Indicator_Diamond
     tiling=NoTile
     imageSize=<64 64>
     margin=<ltrb 0 0 0 0>
     outlineSettings {
      bUseBrushTransparency=false
      roundingType=HalfHeightRadius
      width=0
      color {colorUseRule=UseColor_Specified, specifiedColor=<rgba 0 0 0 0>}
      cornerRadii=<0 0 0 0>
     }
     tintColor {colorUseRule=UseColor_Specified, specifiedColor=<rgba 1 1 1 1>}
     uVRegion {bIsValid=false, max=<0 0>, min=<0 0>}
    }
    renderTransform {angle=0, scale=<1 1>, shear=<0 0>, translation=<8 0>}
   }
   slot {class=ScaleBoxSlot, properties {horizontalAlignment=HAlign_Left}}
  }
  UEFN_TextBlock_C_1 {
   widgetClassPath=&UEFN_TextBlock
   parent=GridPanel_90
   widget=UEFN_TextBlock_C_1
   properties {
    text=Banner
    font {
     bForceMonospaced=false
     bMaterialIsStencil=false
     fontMaterial=null
     fontObject=&BurbankSmall
     letterSpacing=0
     monospacedWidth=1
     size=15
     skewAmount=0
     typefaceFontName=Black
     outlineSettings {
      bApplyOutlineToDropShadows=false
      bMiteredCorners=false
      bSeparateFillAlpha=false
      outlineMaterial=null
      outlineSize=0
      outlineColor=<rgba 0 0 0 1>
     }
    }
   }
   slot {class=GridSlot, properties {columnSpan=1, row=2, rowSpan=1, verticalAlignment=VAlign_Center, padding=<ltrb 5 5 5 5>}}
  }
  Banner_Primary {
   widgetClassPath=&Image
   bIsVariable=true
   parent=GridPanel_90
   widget=Banner_Primary
   slot {class=GridSlot, properties {column=1, columnSpan=1, row=2, rowSpan=1, padding=<ltrb 5 5 5 5>}}
  }
  Banner_Secondary {
   widgetClassPath=&Image
   bIsVariable=true
   parent=GridPanel_90
   widget=Banner_Secondary
   slot {class=GridSlot, properties {column=2, columnSpan=1, row=2, rowSpan=1, padding=<ltrb 5 5 5 5>}}
  }
  ScaleBox_Banner_Icon {
   widgetClassPath=&ScaleBox
   parent=GridPanel_90
   widget=ScaleBox_Banner_Icon
   slot {class=GridSlot, properties {column=3, columnSpan=1, row=2, rowSpan=1}}
  }
  Banner_Icon {
   widgetClassPath=/Game/UI/Frontend/Lobby/Nameplates/WBP_Nameplate_Banner
   bIsVariable=true
   parent=ScaleBox_Banner_Icon
   widget=Banner_Icon
   slot {class=ScaleBoxSlot}
  }
  UEFN_TextBlock_C_2 {
   widgetClassPath=&UEFN_TextBlock
   parent=GridPanel_90
   widget=UEFN_TextBlock_C_2
   properties {
    text="Festival\r\nLocal MP"
    font {
     bForceMonospaced=false
     bMaterialIsStencil=false
     fontMaterial=null
     fontObject=&BurbankSmall
     letterSpacing=0
     monospacedWidth=1
     size=15
     skewAmount=0
     typefaceFontName=Black
     outlineSettings {
      bApplyOutlineToDropShadows=false
      bMiteredCorners=false
      bSeparateFillAlpha=false
      outlineMaterial=null
      outlineSize=0
      outlineColor=<rgba 0 0 0 1>
     }
    }
   }
   slot {class=GridSlot, properties {columnSpan=1, row=3, rowSpan=1, verticalAlignment=VAlign_Center, padding=<ltrb 5 5 5 5>}}
  }
  LocalMP_Color {
   widgetClassPath=&Image
   bIsVariable=true
   parent=GridPanel_90
   widget=LocalMP_Color
   slot {class=GridSlot, properties {column=1, columnSpan=1, row=3, rowSpan=1, padding=<ltrb 5 5 5 5>}}
  }
 ]
}
