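# Player nameplate widget for Pilgrim Quickplay, showing avatar, local identity, name, band leader icon and a collapsed-by-default instrument and difficulty box.
# PreConstruct assigns the borderless avatar material, then at design time previews a sample player's colors, difficulty and instrument.
# Functions swap the underlayer material from an identity color struct, set the instrument icon from a soft material by track type, and show or hide the widget.
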WidgetBlueprint /PilgrimQuickplay/Widgets/Players/WBP_Pilgrim_PlayerIdentity_Nameplate
graphs [
 EventGraph {nodes [
   n1 = Event PreConstruct()
   > SetBrushFromMaterial(
    self=Get PlayerImage!(self:WBP Pilgrim Avatar Icon Object Reference=Get AvatarIcon() -> (AvatarIcon:WBP Pilgrim Avatar Icon Object Reference)) -> (PlayerImage:Image Object Reference)
    Material=/SparksUICommon/Materials/MI_UI_BandMember_Avatar_Master_LocalPlayerWithouBorder
   )
   > Branch(Condition=@n1.IsDesignTime)
   n2 = Get Design Preview Local Player Index() -> (Design Preview Local Player Index:int=0)
   n3 = GetPilgrim LocalPlayerIdentity ColorInfo By Index!(LocalPlayerIndex:int=@n2) -> (Player Identity Color Info:S Pilgrim Local Player Identifier Definition Structure)
   > SetPlayerColorInfo!(execute, self, LocalidColor:S Pilgrim Local Player Identifier Definition Structure=@n3) -> (then)
   > SetPlayerAvatarBackground!(
    execute
    self:WBP Pilgrim Avatar Icon Object Reference=Get AvatarIcon()
    BgColor:Linear Color Structure=Break S Pilgrim Local Player Identifier Definition!(S_Pilgrim_LocalPlayerIdentifierDefinition:"S Pilgrim Local Player Identifier Definition Structure (by ref)"=@n3) -> (IdentityColor_5_#1:Linear Color Structure) {is=title K2Node_BreakStruct}
   ) -> (then)
   > SetDifficulty!(
    execute
    self:WBP Pilgrim Results Difficulty Object Reference=Get WBP_Pilgrim_Results_Difficulty() -> (WBP_Pilgrim_Results_Difficulty:WBP Pilgrim Results Difficulty Object Reference)
    Difficulty:EPilgrimSongDifficulty Enum=DifficultyExpert
   ) -> (then)
   > Set Instrument Type!(execute, self, TrackType:EPilgrimTrackType Enum=TrackDrum) -> (then) {is=title K2Node_CallFunction}
   > SetLocalPlayerIndex!(
    execute
    self:WBP Pilgrim Player Identity Local Identity Label Object Reference=Get PlayerIdentity_LocalIdentity_Label() -> (PlayerIdentity_LocalIdentity_Label:WBP Pilgrim Player Identity Local Identity Label Object Reference)
    LocalPlayerIndex:int=@n2
   ) -> (then)
   > SetLocalPlayerIndex(
    self:WBP Pilgrim Player Identity Local Identity Icon Object Reference=Get PlayerIdentity_LocalIdentity_Icon() -> (PlayerIdentity_LocalIdentity_Icon:WBP Pilgrim Player Identity Local Identity Icon Object Reference)
    LocalPlayerIndex=@n2
   )
  ]}
 HidePlayerDisplay {nodes [
   HidePlayerDisplay() {is=title K2Node_FunctionEntry}
   > SetVisibility(self=Self-Reference(), InVisibility=Hidden)
  ]}
 Set Instrument Type {nodes [
   n1 = Set Instrument Type!() -> (then, TrackType:EPilgrimTrackType Enum=TrackGuitar) {is=title K2Node_FunctionEntry}
   > SetBrushFromSoftMaterial!(
    execute
    self:Image Object Reference=Get InstrumentImage() -> (InstrumentImage:Fort Lazy Image Object Reference)
    SoftMaterial:Material Interface Soft Object Reference=GetIconMaterialByInstrument!(
     Instrument:EPilgrimTrackType Enum=@n1
     SongMetaData:Pilgrim Song Metadata Object Reference
     SparksInstrument:ESparksInstrumentType Enum=None
     bUseKeytarIcon:bool=true
     UseProIcons:bool=true
    ) -> (IconMaterial:Material Instance Soft Object Reference)
   ) -> (then)
  ]}
 SetPlayerColorInfo {nodes [
   n1 = SetPlayerColorInfo!() -> (then, LocalidColor:S Pilgrim Local Player Identifier Definition Structure) {is=title K2Node_FunctionEntry}
   > SetBrushFromMaterial(
    self=Get PlayerUnderlayer() -> (PlayerUnderlayer:Image Object Reference)
    Material=Break S Pilgrim Local Player Identifier Definition!(S_Pilgrim_LocalPlayerIdentifierDefinition=@n1) -> (
     IdentityLabel_16_#2:Text
     PlayerColorUnderlayerM_11_#3:Material Instance Object Reference
    ).PlayerColorUnderlayerM_11_#3
   )
  ]}
 ShowPlayerDisplay {nodes [
   ShowPlayerDisplay() {is=title K2Node_FunctionEntry}
   > SetVisibility(self=Self-Reference(), InVisibility=SelfHitTestInvisible)
  ]}
]
variables [
 Design Preview Local Player Index {type=integer, default=2}
 MyPlayerState {type=PlayerState, default=null}
 PlayerIconRetryCount {type=integer, default=10}
 PlayerIconRetryCount_Default {type=integer, default=10}
]
widgets {
 info {inheritedWidgetCount=0, parentClass=/Script/PilgrimCoreUI.PilgrimFixedOwnerWidget, rootWidgetClass=&GridPanel, widgetCount=16}
 widgets [
  GridPanel_0 {widgetClassPath=&GridPanel, widget=GridPanel_0, properties {columnFill [1], rowFill [1]}}
  PlayerUnderlayer {
   widgetClassPath=&Image
   bIsVariable=true
   parent=GridPanel_0
   widget=PlayerUnderlayer
   properties {brush {
     bIsDynamicallyLoaded=false
     drawAs=Image
     imageType=FullColor
     mirroring=NoMirror
     resourceName=None
     resourceObject=/PilgrimCore/Widgets/Materials/LocalPlayerMaterials/M_PilgrimLocalPlayer_Underlayer_P3
     tiling=NoTile
     imageSize=<32 32>
     margin=<ltrb 0 0 0 0>
     outlineSettings {
      bUseBrushTransparency=false
      roundingType=HalfHeightRadius
      width=0
      color {colorUseRule=UseColor_Specified, specifiedColor=<rgba 0 0 0 0>}
      cornerRadii=<0 0 0 0>
     }
     tintColor {colorUseRule=UseColor_Specified, specifiedColor=<rgba 1 1 1 1>}
     uVRegion {bIsValid=false, max=<0 0>, min=<0 0>}
    }}
   slot {class=GridSlot, properties {columnSpan=1, rowSpan=1}}
  }
  PlayerID {
   widgetClassPath=&HorizontalBox
   parent=GridPanel_0
   widget=PlayerID
   slot {class=GridSlot, properties {columnSpan=1, layer=1, rowSpan=1}}
  }
  AvatarIcon {
   widgetClassPath=/PilgrimCore/Widgets/Player/WBP_Pilgrim_AvatarIcon
   bIsVariable=true
   parent=PlayerID
   widget=AvatarIcon
   properties {playerIconSizeOverride=<46 46>}
   slot {class=HorizontalBoxSlot, properties {padding=<ltrb 3 3 0 3>}}
  }
  Spacer_62 {widgetClassPath=&Spacer, parent=PlayerID, widget=Spacer_62, properties {size=<6 1>}, slot {class=HorizontalBoxSlot}}
  PlayerIdentity_LocalIdentity_Label {
   widgetClassPath=/PilgrimCore/Widgets/Player/WBP_Pilgrim_PlayerIdentity_LocalIdentity_Label
   bIsVariable=true
   parent=PlayerID
   widget=PlayerIdentity_LocalIdentity_Label
   properties {fontSizeOverride=18}
   slot {class=HorizontalBoxSlot, properties {verticalAlignment=VAlign_Center, padding=<ltrb 0 0 2 0>}}
  }
  PlayerIdentity_LocalIdentity_Icon {
   widgetClassPath=/PilgrimCore/Widgets/Player/WBP_Pilgrim_PlayerIdentity_LocalIdentity_Icon
   bIsVariable=true
   parent=PlayerID
   widget=PlayerIdentity_LocalIdentity_Icon
   properties {visibility=Collapsed, iconSizeOverride=<18 18>}
   slot {class=HorizontalBoxSlot, properties {verticalAlignment=VAlign_Center}}
  }
  Spacer {widgetClassPath=&Spacer, parent=PlayerID, widget=Spacer, properties {size=<12 1>}, slot {class=HorizontalBoxSlot}}
  PlayerName {
   widgetClassPath=&CommonTextBlock
   bIsVariable=true
   parent=PlayerID
   widget=PlayerName
   properties {
    minDesiredWidth=200
    scrollStyle=&TextScrollStyle-Base_C
    text=John-Philip_Dev1
    font {
     bForceMonospaced=false
     bMaterialIsStencil=false
     fontMaterial=null
     fontObject=&HeadingNow
     letterSpacing=0
     monospacedWidth=1
     size=14.25
     skewAmount=0
     typefaceFontName=75 Medium
     outlineSettings {
      bApplyOutlineToDropShadows=false
      bMiteredCorners=false
      bSeparateFillAlpha=false
      outlineMaterial=null
      outlineSize=0
      outlineColor=<rgba 0 0 0 1>
     }
    }
    shadowOffset=<0 0>
    strikeBrush {
     bIsDynamicallyLoaded=false
     drawAs=NoDrawType
     imageType=NoImage
     mirroring=NoMirror
     resourceName=None
     resourceObject=null
     tiling=NoTile
     imageSize=<32 32>
     margin=<ltrb 0 0 0 0>
     outlineSettings {
      bUseBrushTransparency=false
      roundingType=HalfHeightRadius
      width=0
      color {colorUseRule=UseColor_Specified, specifiedColor=<rgba 0 0 0 0>}
      cornerRadii=<0 0 0 0>
     }
     tintColor {colorUseRule=UseColor_Specified, specifiedColor=<rgba 1 1 1 1>}
     uVRegion {bIsValid=false, max=<0 0>, min=<0 0>}
    }
   }
   slot {class=HorizontalBoxSlot, properties {verticalAlignment=VAlign_Center, size {sizeRule=Fill, value=1}}}
  }
  Spacer_1 {widgetClassPath=&Spacer, parent=PlayerID, widget=Spacer_1, properties {size=<2 0>}, slot {class=HorizontalBoxSlot}}
  BandLeaderIcon {
   widgetClassPath=/PilgrimCore/Widgets/Player/WBP_Pilgrim_BandLeaderIcon
   bIsVariable=true
   parent=PlayerID
   widget=BandLeaderIcon
   slot {class=HorizontalBoxSlot, properties {horizontalAlignment=HAlign_Right, verticalAlignment=VAlign_Center, padding=<ltrb 0 0 10 2>}}
  }
  InstrumentBox {
   widgetClassPath=&HorizontalBox
   parent=PlayerID
   widget=InstrumentBox
   properties {visibility=Collapsed}
   slot {class=HorizontalBoxSlot}
  }
  InstrumentImage {
   widgetClassPath=&FortLazyImage
   bIsVariable=true
   parent=InstrumentBox
   widget=InstrumentImage
   properties {brush {
     bIsDynamicallyLoaded=false
     drawAs=Image
     imageType=NoImage
     mirroring=NoMirror
     resourceName=None
     resourceObject=/PilgrimQuickplay/Widgets/Materials/MI_UI_Icon_Instrument_Type_Guitar
     tiling=NoTile
     imageSize=<52 52>
     margin=<ltrb 0 0 0 0>
     outlineSettings {
      bUseBrushTransparency=false
      roundingType=HalfHeightRadius
      width=0
      color {colorUseRule=UseColor_Specified, specifiedColor=<rgba 0 0 0 0>}
      cornerRadii=<0 0 0 0>
     }
     tintColor {colorUseRule=UseColor_Specified, specifiedColor=<rgba 1 1 1 1>}
     uVRegion {bIsValid=false, max=<0 0>, min=<0 0>}
    }}
   slot {class=HorizontalBoxSlot, properties {horizontalAlignment=HAlign_Right, verticalAlignment=VAlign_Center, padding=<ltrb 0 0 6 0>}}
  }
  SC_Difficulty {
   widgetClassPath=&ScaleBox
   parent=InstrumentBox
   widget=SC_Difficulty
   properties {stretch=UserSpecified, userSpecifiedScale=0.5}
   slot {class=HorizontalBoxSlot, properties {horizontalAlignment=HAlign_Right, verticalAlignment=VAlign_Center, padding=<ltrb 0 0 6 0>}}
  }
  WBP_Pilgrim_Results_Difficulty {
   widgetClassPath=/PilgrimQuickplay/Widgets/Results/WBP_Pilgrim_Results_Difficulty
   bIsVariable=true
   parent=SC_Difficulty
   widget=WBP_Pilgrim_Results_Difficulty
   properties {renderTransform {angle=0, scale=<1.5 1.5>, shear=<0 0>, translation=<0 0>}}
   slot {class=ScaleBoxSlot}
  }
  Spacer_2 {widgetClassPath=&Spacer, parent=InstrumentBox, widget=Spacer_2, properties {size=<10 1>}, slot {class=HorizontalBoxSlot}}
 ]
}
